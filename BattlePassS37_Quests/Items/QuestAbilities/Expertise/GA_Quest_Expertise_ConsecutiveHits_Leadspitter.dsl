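; Passive quest ability that tracks consecutive Leadspitter hits for a Season 37 expertise quest.
; Weapon-fire gameplay events flag a shot in flight, or reset the streak if the previous shot is still unresolved, treating it as a miss.
; A filtered EventProcessor hit event clears the flag, increments and debug-prints the count, and broadcasts the Minigun ConsecutiveHit custom verb from the second hit onward.
; Ending the ability cancels both listeners.
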(event Ability|EventFailedToActivatePassiveAbility
  (params (ActorInfo :type Gameplay%20Ability%20Actor%20Info%20Structure))
  (|CallParentFunction ActorInfo))

(event Ability|EventOnEndAbility
  (params (bWasCancelled :type Boolean))
  (|CallParentFunction bWasCancelled)
  (AsyncAction|Cancel (Variables|Default|GetAsyncActionA))
  (EventProcessorSubsystem|StopListeningToFilteredEvent (Variables|Default|GetHitEvent_A)))

(event Ability|EventActivateAbility
  (|CallParentFunction)
  (bind (_asynctaskproxy _payload) (Ability|Async|WaitGameplayEventToActor (Ability|GetOwningActorfromActorInfo) "(TagName=\"Event.Weapons.Fire\")")
    (:then
      (Variables|Default|SetAsyncActionA _asynctaskproxy)
      (bind _returnvalue (EventProcessorSubsystem|StartListeningToFilteredEvent 0 (Ability|GetOwningActorfromActorInfo) (Variables|Default|GetVerb%20Filter) false (delegate Custom|HitEvent)))
      (Variables|Default|SetHitEvent_A _returnvalue))
    (:EventReceived
      (Variables|Default|SetCurrentPayload _payload)
      (CallFunction|WeaponFired))))

(event Custom|HitEvent
  (if (Variables|Default|GetShotInFlight)
    (Math|Integer|IncrementInt (Variables|Default|GetConsecutiveHitCount)
      (:%20%20
        (Variables|Default|SetShotInFlight false)
        (Development|PrintString (Utilities|String|Append "Shots Hit: " (Utilities|String|ToString(Integer) (Variables|Default|GetConsecutiveHitCount))) false false "(R=1.000000,G=0.360929,B=0.000000,A=1.000000)" 5.0)
        (if (> (Variables|Default|GetConsecutiveHitCount) 1)
          (CustomVerb|BroadcastCustomVerbMessage (CustomVerb|MakeCustomVerbMessage (Ability|GetOwningActorfromActorInfo) "(TagName=\"CustomVerb.Quest.S37.Expertise.Minigun.ConsecutiveHit\")")))))))

(event Custom|WeaponFired
  (if (Variables|Default|GetShotInFlight)
    (Variables|Default|SetConsecutiveHitCount 0)
    (else
      (Variables|Default|SetShotInFlight true))))
